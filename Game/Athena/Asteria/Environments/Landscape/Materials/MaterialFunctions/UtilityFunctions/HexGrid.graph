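# Material function that builds a hex grid for Asteria landscape materials, outputting three grid coordinates and matching blend weights.
# UVs are skewed into a triangular lattice and scaled by HexSize, then floor and frac yield cell IDs and in-triangle weights.
# A mod-3 mask from the cell index keeps weights and vertex IDs in consistent channels across triangles.
# Weights are sharpened by a power and normalized into GradientsRGBA, alongside Grid1, Grid2 and Grid3.
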MaterialFunction /Game/Athena/Asteria/Environments/Landscape/Materials/MaterialFunctions/UtilityFunctions/HexGrid
graph {nodes [
  n1 = FunctionInput(Preview=Multiply(A=TextureCoordinate(), B=ScalarParameter() {defaultValue=12, group=HexGrid, parameterName=HG_GridTiling})) {inputName=UVCoords, inputType=FunctionInput_Vector2}
  n2 = ComponentMask(None=@n1) {r=true}
  n3 = ComponentMask(None=@n1) {g=true}
  n4 = Divide(A=Constant() {r=1.732})
  n5 = Multiply(A=@n3, B=Divide(B=@n4) {constA=0.5})
  n6 = Divide(B=@n4)
  n7 = Divide(
   A=AppendVector(A=Subtract(A=@n2, B=@n5), B=Multiply(A=@n3, B=@n6) {constB=1.732})
   B=FunctionInput(Preview=ScalarParameter() {defaultValue=1, group=HexGrid, parameterName=HG_HexSize}) {bUsePreviewValueAsDefault=true, inputName=HexSize, inputType=FunctionInput_Scalar, sortPriority=1}
  )
  n8 = NamedRerouteDeclaration(None=Floor(None=@n7)) {name=UV_Floor, nodeColor=<rgba 1 0 0.1641 1>}
  n9 = Subtract(A=ComponentMask(None=@n8) {r=true}, B=ComponentMask(None=@n8) {g=true})
  n10 = Constant3Vector() {constant=<rgba 0 1 2 0>}
  n11 = Add(A=Multiply(A=Add(A=@n9, B=@n10), B=Divide() {constB=3})) {constB=1.667}
  n12 = NamedRerouteDeclaration(None=Round(None=Frac(None=@n11))) {name=Round_Result, nodeColor=<rgba 1 0.0234 0 1>}
  n13 = MaterialFunctionCall(Color=@n12) {materialFunction=&SplitComponents}
  n14 = MaterialFunctionCall(R=@n13.B, G=@n13.B, B=@n13.B) {materialFunction=&AppendMany}
  n15 = NamedRerouteUsage()
  n16 = MaterialFunctionCall(R=@n13.G, G=@n13.G, B=@n13.G) {materialFunction=&AppendMany}
  n17 = NamedRerouteUsage()
  n18 = NamedRerouteUsage()
  n19 = Power(
   Base=MaterialFunctionCall(R=DotProduct(A=@n14.RGB, B=@n15), G=DotProduct(A=@n16.RGB, B=@n17), B=DotProduct(A=@n12, B=@n18)) {materialFunction=&AppendMany}.RGB
   Exp=FunctionInput(Preview=ScalarParameter() {defaultValue=6, group=HexGrid, parameterName=HG_HexTileEdgeSharpen}) {bUsePreviewValueAsDefault=true, inputName=HexGradientSharpness, inputType=FunctionInput_Scalar, sortPriority=2}
  )
  n20 = DotProduct(A=@n19, B=Constant() {r=1})
  FunctionOutput(None=NamedRerouteDeclaration(None=Divide(A=@n19, B=@n20)) {name=Gradients, nodeColor=<rgba 1 0 0.6094 1>}) {outputName=GradientsRGBA}
  n21 = NamedRerouteUsage()
  n22 = NamedRerouteUsage()
  n23 = Constant() {r=1}
  n24 = Constant()
  n25 = MaterialFunctionCall(Float3=Multiply(A=@n21, B=If(A=@n22, "A > B"=@n23, "A == B"=@n24, "A < B"=@n24))) {materialFunction=&BreakOutFloat3Components}
  n26 = NamedRerouteUsage()
  n27 = Add(A=@n26, B=@n25.B)
  n28 = MaterialFunctionCall(Float3=@n21) {materialFunction=&BreakOutFloat3Components}
  n29 = AppendVector(A=@n28.R, B=@n28.R)
  FunctionOutput(None=NamedRerouteDeclaration(None=Add(A=@n27, B=@n29)) {name=Hex1, nodeColor=<rgba 1 0 0.1641 1>}) {outputName=Grid1}
  n30 = Add(A=@n26, B=@n25.G)
  n31 = AppendVector(A=@n28.B, B=@n28.B)
  FunctionOutput(None=NamedRerouteDeclaration(None=Add(A=@n30, B=@n31)) {name=Hex2, nodeColor=<rgba 1 0 0.8672 1>}) {outputName=Grid2}
  n32 = Add(A=@n26, B=@n25.R)
  n33 = AppendVector(A=@n28.G, B=@n28.G)
  FunctionOutput(None=NamedRerouteDeclaration(None=Add(A=@n32, B=@n33)) {name=Hex3, nodeColor=<rgba 0 1 0.8594 1>}) {outputName=Grid3}
  n34 = Frac(None=@n7)
  n35 = ComponentMask(None=@n34) {r=true}
  n36 = ComponentMask(None=@n34) {g=true}
  n37 = NamedRerouteDeclaration(None=Subtract(A=Add(A=@n35, B=@n36))) {name=SubtractResult, nodeColor=<rgba 0 0.6953 1 1>}
  n38 = OneMinus(None=MaterialFunctionCall(XY=@n34) {materialFunction=&Swizzle}.YX)
  NamedRerouteDeclaration(None=AppendVector(A=Abs(None=@n37), B=If(A=@n37, "A > B"=@n38, "A == B"=@n34, "A < B"=@n34))) {name=Grid_Result, nodeColor=<rgba 0.4297 0 1 1>}
 ]}
