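# HexGridHash is a landscape utility material function that hashes a 2D coordinate into four pseudo-random values, presumably for hex grid cells.
# Four dot products of rearranged X/Y components with large constant vectors are packed into a float4.
# The classic sine-hash follows: Sine, multiply by a large constant offset by the Seed input, then Frac.
# UV defaults to a ceiled, scaled texture coordinate; a leftover MakeFloat3 node is unused.
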MaterialFunction /Game/Athena/Asteria/Environments/Landscape/Materials/MaterialFunctions/UtilityFunctions/HexGridHash
graph {nodes [
  n1 = TextureCoordinate()
  n2 = FunctionInput(Preview=Ceil(None=Multiply(A=@n1) {constB=10})) {bUsePreviewValueAsDefault=true, inputName=UV, inputType=FunctionInput_Vector2}
  n3 = ComponentMask(None=@n2) {r=true}
  n4 = ComponentMask(None=@n2) {g=true}
  n5 = DotProduct(
   A=MaterialFunctionCall(R=@n3, G=@n4, B=@n3) {materialFunction=&AppendMany}.RGB
   B=Constant3Vector() {constant=<rgba 574.9868 284.3817 847.0239 0>}
  )
  n6 = DotProduct(
   A=MaterialFunctionCall(R=@n4, G=@n3, B=@n4) {materialFunction=&AppendMany}.RGB
   B=Constant3Vector() {constant=<rgba 235.9474 724.3587 485.192 0>}
  )
  n7 = DotProduct(
   A=MaterialFunctionCall(R=@n4, G=@n4, B=@n3) {materialFunction=&AppendMany}.RGB
   B=Constant3Vector() {constant=<rgba 639.9105 178.392 597.2471 0>}
  )
  n8 = MaterialFunctionCall(R=@n3, G=@n3, B=@n4) {materialFunction=&AppendMany}
  n9 = Constant3Vector() {constant=<rgba 945.7268 286.5635 584.8473 0>}
  n10 = Sine(None=MaterialFunctionCall(X=@n5, Y=@n6, Z=@n7, A=DotProduct(A=@n8.RGB, B=@n9)) {materialFunction=&MakeFloat4}.Result)
  n11 = Add(
   A=Constant() {r=43758.5469}
   B=FunctionInput(Preview=Constant() {r=2}) {bUsePreviewValueAsDefault=true, inputName=Seed, inputType=FunctionInput_Scalar}
  )
  FunctionOutput(None=Frac(None=Multiply(A=@n10, B=@n11)))
  MaterialFunctionCall(X=@n5, Y=@n6, Z=@n7) {materialFunction=&MakeFloat3}
 ]}
