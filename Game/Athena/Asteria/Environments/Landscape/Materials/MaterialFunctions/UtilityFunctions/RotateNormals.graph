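# RotateNormals is a landscape utility material function that rotates a normal's XY components by a scalar Rotation angle.
# It applies a standard 2D rotation, combining the sine and cosine of the angle with the R and G channels.
# A 2π period on both trig nodes makes the angle radians.
# The rotated XY is re-appended to the untouched B and A channels, producing a four-component output.
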MaterialFunction /Game/Athena/Asteria/Environments/Landscape/Materials/MaterialFunctions/UtilityFunctions/RotateNormals
graph {nodes [
  n1 = FunctionInput() {bUsePreviewValueAsDefault=true, inputName=Rotation, inputType=FunctionInput_Scalar}
  n2 = Cosine(None=@n1) {period=6.2832}
  n3 = FunctionInput() {bUsePreviewValueAsDefault=true, inputName=Normal, inputType=FunctionInput_Vector4}
  n4 = MaterialFunctionCall(Float2=ComponentMask(None=@n3) {g=true, r=true}) {materialFunction=&BreakOutFloat2Components}
  n5 = Sine(None=@n1) {period=6.2832}
  n6 = Subtract(A=Multiply(A=@n2, B=@n4.R), B=Multiply(A=@n5, B=@n4.G))
  n7 = Add(A=Multiply(A=@n5, B=@n4.R), B=Multiply(A=@n2, B=@n4.G))
  FunctionOutput(None=AppendVector(A=AppendVector(A=@n6, B=@n7), B=ComponentMask(None=@n3) {a=true, b=true}))
 ]}
