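# TransformUV is a landscape material utility function that applies hash-driven random rotation, offset and scale to UVs per grid cell.
# HexGridHash takes a seed and grid input, defaulting to the ceiling of the UVs, and its channels feed named reroutes for lerps between min/max limits.
# UVs are shifted by half, rotated, offset, then scaled by center.
# A second output, RotateValue, returns two pi minus a rerouted value.
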MaterialFunction /Game/Athena/Asteria/Environments/Landscape/Materials/MaterialFunctions/UtilityFunctions/TransformUV
graph {nodes [
  n1 = FunctionInput(Preview=Multiply(A=TextureCoordinate(), B=ScalarParameter() {defaultValue=1, group=TransformUV, parameterName=TilingVal})) {bUsePreviewValueAsDefault=true, inputName=UVCoordinate, inputType=FunctionInput_Vector2}
  n2 = MaterialFunctionCall!(Angle=NamedRerouteUsage(), UV/2DVector=Add(A=@n1) {constB=-0.5}) {materialFunction=&/Rotate2D}
  n3 = NamedRerouteUsage()
  FunctionOutput(None=MaterialFunctionCall(UVs=Add(A=@n2.Result, B=@n3), Texture Scale=NamedRerouteUsage()) {materialFunction=&ScaleUVsByCenter}.UVs)
  FunctionOutput(None=Subtract(A=Constant() {r=6.2832}, B=NamedRerouteUsage())) {outputName=RotateValue, sortPriority=1}
  n4 = ScalarParameter() {defaultValue=2, group=TransformUV, parameterName=Seed}
  n5 = Ceil(None=@n1)
  n6 = MaterialFunctionCall(Float4=MaterialFunctionCall!(
    Seed=FunctionInput(Preview=@n4) {bUsePreviewValueAsDefault=true, inputName=Seed, inputType=FunctionInput_Scalar, sortPriority=5}
    UV=FunctionInput(Preview=@n5) {bUsePreviewValueAsDefault=true, inputName=HexGridInput, inputType=FunctionInput_Scalar, sortPriority=4}
   ) {materialFunction=&/HexGridHash}.Result) {materialFunction=&BreakOutFloat4Components}
  NamedRerouteDeclaration(None=@n6.R) {name=R_HashValue, nodeColor=<rgba 0.4297 0 1 1>}
  NamedRerouteDeclaration(None=@n6.G) {name=G_HashValue, nodeColor=<rgba 0.4062 0 1 1>}
  NamedRerouteDeclaration(None=@n6.B) {name=B_HashValue, nodeColor=<rgba 0.4062 0 1 1>}
  n7 = FunctionInput(Preview=ScalarParameter() {defaultValue=-90, group=TransformUV, parameterName=MinRotation}) {bUsePreviewValueAsDefault=true, inputName="MinRotation(Degrees)", inputType=FunctionInput_Scalar, sortPriority=3}
  n8 = FunctionInput(Preview=ScalarParameter() {defaultValue=90, group=TransformUV, parameterName=MaxRotation}) {bUsePreviewValueAsDefault=true, inputName="MaxRotation(Degrees)", inputType=FunctionInput_Scalar, sortPriority=3}
  n9 = NamedRerouteUsage()
  NamedRerouteDeclaration(None=MaterialFunctionCall!(Degrees=LinearInterpolate(A=@n7, B=@n8, Alpha=@n9)) {materialFunction=/Engine/Functions/Engine_MaterialFunctions01/Math/DegreesToRadians}.Radians) {name=HexRotationValue, nodeColor=<rgba 1 0.5859 0 1>}
  n10 = ComponentMask(None=VectorParameter() {group=TransformUV, parameterName=MinOffset, defaultValue=<rgba -1 -1 0 1>}.RGB) {g=true, r=true}
  n11 = ComponentMask(None=VectorParameter() {group=TransformUV, parameterName=MaxOffset, defaultValue=<rgba 1 1 0 1>}.RGB) {g=true, r=true}
  NamedRerouteDeclaration(None=LinearInterpolate(
    A=FunctionInput(Preview=@n10) {bUsePreviewValueAsDefault=true, inputName=MinOffset, inputType=FunctionInput_Vector2, sortPriority=2}
    B=FunctionInput(Preview=@n11) {bUsePreviewValueAsDefault=true, inputName=MaxOffset, inputType=FunctionInput_Vector2, sortPriority=2}
    Alpha=NamedRerouteUsage()
   )) {name=HexOffsetValue, nodeColor=<rgba 0 1 0.1562 1>}
  n12 = ScalarParameter() {defaultValue=1, group=TransformUV, parameterName=MinScale}
  n13 = ScalarParameter() {defaultValue=1.5, group=TransformUV, parameterName=MaxScale}
  NamedRerouteDeclaration(None=LinearInterpolate(
    A=FunctionInput(Preview=@n12) {bUsePreviewValueAsDefault=true, inputName=MinScale, inputType=FunctionInput_Scalar, sortPriority=1}
    B=FunctionInput(Preview=@n13) {bUsePreviewValueAsDefault=true, inputName=MaxScale, inputType=FunctionInput_Scalar, sortPriority=1}
    Alpha=NamedRerouteUsage()
   )) {name=HexScaleValue, nodeColor=<rgba 0.7578 0 1 1>}
  NamedRerouteDeclaration(None=@n6.A) {name=A_HashValue, nodeColor=<rgba 0.4062 0 1 1>}
 ]}
