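# Landscape utility material function that samples a normal map with hex-grid tiling to break up visible repetition.
# HexGrid drives three randomly scaled, offset and rotated UV sets; each sample, with optional mip bias, has its normal rotated to match, then is gradient-weighted and summed.
# A cheaper single tiled sample forms the low-end path.
# The hex blend is used only at Epic quality on default or deferred shading, unless forceLowQuality overrides it.
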MaterialFunction /Game/Athena/Asteria/Environments/Landscape/Materials/MaterialFunctions/UtilityFunctions/Landscape_NormalTextureSample
graph {
 properties {bExposeToLibrary=true}
 nodes [
  n1 = NamedRerouteUsage()
  n2 = QualitySwitch(Default=@n1, Low=@n1, High=@n1, Medium=@n1, Epic=NamedRerouteUsage())
  n3 = StaticBool()
  FunctionOutput(None=StaticSwitch(
    True=@n1
    False=ShadingPathSwitch(Default=@n2, Deferred=@n2, Forward=@n1, Mobile=@n1)
    Value=FunctionInput(Preview=@n3) {bUsePreviewValueAsDefault=true, inputName=forceLowQuality, inputType=FunctionInput_StaticBool, sortPriority=20}
   ))
  NamedRerouteDeclaration(None=FunctionInput(Preview=StaticBool()) {bUsePreviewValueAsDefault=true, inputName=UseMipBias, inputType=FunctionInput_StaticBool}) {name=UseMipBias, nodeColor=<rgba 1 0 0.9141 1>}
  n4 = FunctionInput(Preview=Multiply(A=TextureCoordinate()) {constB=12}) {bUsePreviewValueAsDefault=true, inputName=LandscapeCoordinates, inputType=FunctionInput_Vector2}
  n5 = MaterialFunctionCall!(
   UVCoords=FunctionInput(Preview=@n4) {bUsePreviewValueAsDefault=true, inputName=HexGridUVInput, inputType=FunctionInput_Vector2, sortPriority=1}
   HexSize=FunctionInput(Preview=Constant() {r=1}) {bUsePreviewValueAsDefault=true, inputName=HexGridSize, inputType=FunctionInput_Scalar, sortPriority=1}
   HexGradientSharpness=FunctionInput(Preview=Constant() {r=6}) {bUsePreviewValueAsDefault=true, inputName=HexGridGradientSharpness, inputType=FunctionInput_Scalar, sortPriority=1}
  ) {materialFunction=&/HexGrid}
  n6 = MaterialFunctionCall!(
   UVCoordinate=@n4
   MinScale=NamedRerouteUsage()
   MaxScale=NamedRerouteUsage()
   MinOffset=NamedRerouteUsage()
   MaxOffset=NamedRerouteUsage()
   "MaxRotation(Degrees)"=NamedRerouteUsage()
   "MinRotation(Degrees)"=NamedRerouteUsage()
   HexGridInput=@n5.Grid1
   Seed
  ) {materialFunction=&/TransformUV}
  n7 = NamedRerouteUsage()
  n8 = NamedRerouteDeclaration(None=FunctionInput(Preview=ScalarParameter() {defaultValue=3, group=HexGrid, parameterName=ViewMipBIas}) {bUsePreviewValueAsDefault=true, inputName=MipBias, inputType=FunctionInput_Scalar}) {name=MipBias, nodeColor=<rgba 0.7812 1 0 1>}
  n9 = TextureSample!(UVs=@n6.Result, Tex=@n7, Bias=@n8, Apply View MipBias) {mipValueMode=TMVM_MipBias, samplerSource=SSM_Wrap_WorldGroupSettings, samplerType=SAMPLERTYPE_Normal}
  n10 = TextureSample!(UVs=@n6.Result, Tex=@n7, Apply View MipBias) {samplerSource=SSM_Wrap_WorldGroupSettings, samplerType=SAMPLERTYPE_Normal}
  n11 = NamedRerouteUsage()
  n12 = Multiply(
   A=MaterialFunctionCall!(Rotation=@n6.RotateValue, Normal=StaticSwitch(True=@n9.RGBA, False=@n10.RGBA, Value=@n11)) {materialFunction=&/RotateNormals}.Result
   B=NamedRerouteUsage()
  )
  n13 = MaterialFunctionCall(
   UVCoordinate=@n4
   MinScale=NamedRerouteUsage()
   MaxScale=NamedRerouteUsage()
   MinOffset=NamedRerouteUsage()
   MaxOffset=NamedRerouteUsage()
   "MaxRotation(Degrees)"=NamedRerouteUsage()
   "MinRotation(Degrees)"=NamedRerouteUsage()
   HexGridInput=@n5.Grid2
  ) {materialFunction=&TransformUV}
  n14 = NamedRerouteUsage()
  n15 = TextureSample!(UVs=@n13.Result, Tex=@n14, Bias=@n8, Apply View MipBias) {mipValueMode=TMVM_MipBias, samplerSource=SSM_Wrap_WorldGroupSettings, samplerType=SAMPLERTYPE_Normal}
  n16 = TextureSample!(UVs=@n13.Result, Tex=@n14, Apply View MipBias) {samplerSource=SSM_Wrap_WorldGroupSettings, samplerType=SAMPLERTYPE_Normal}
  n17 = NamedRerouteUsage()
  n18 = Multiply(
   A=MaterialFunctionCall(Rotation=@n13.RotateValue, Normal=StaticSwitch(True=@n15.RGBA, False=@n16.RGBA, Value=@n17)) {materialFunction=&RotateNormals}.Result
   B=NamedRerouteUsage()
  )
  n19 = MaterialFunctionCall(
   UVCoordinate=@n4
   MinScale=NamedRerouteUsage()
   MaxScale=NamedRerouteUsage()
   MinOffset=NamedRerouteUsage()
   MaxOffset=NamedRerouteUsage()
   "MaxRotation(Degrees)"=NamedRerouteUsage()
   "MinRotation(Degrees)"=NamedRerouteUsage()
   HexGridInput=@n5.Grid3
  ) {materialFunction=&TransformUV}
  n20 = NamedRerouteUsage()
  n21 = MaterialFunctionCall(
   Rotation=@n19.RotateValue
   Normal=StaticSwitch(
    True=TextureSample!(UVs=@n19.Result, Tex=@n20, Bias=@n8, Apply View MipBias) {mipValueMode=TMVM_MipBias, samplerSource=SSM_Wrap_WorldGroupSettings, samplerType=SAMPLERTYPE_Normal}.RGBA
    False=TextureSample!(UVs=@n19.Result, Tex=@n20, Apply View MipBias) {samplerSource=SSM_Wrap_WorldGroupSettings, samplerType=SAMPLERTYPE_Normal}.RGBA
    Value=NamedRerouteUsage()
   )
  ) {materialFunction=&RotateNormals}
  n22 = NamedRerouteUsage()
  NamedRerouteDeclaration(None=Add(A=Add(A=@n12, B=@n18), B=Multiply(A=@n21.Result, B=@n22))) {name=HighEnd_Output, nodeColor=<rgba 0 0.4141 1 1>}
  n23 = MaterialFunctionCall(Float3=@n5.GradientsRGBA) {materialFunction=&BreakOutFloat3Components}
  NamedRerouteDeclaration(None=@n23.G) {name=HexGradient_02, nodeColor=<rgba 1 0 0.8672 1>}
  NamedRerouteDeclaration(None=@n23.B) {name=HexGradient_03, nodeColor=<rgba 1 0 0.8672 1>}
  NamedRerouteDeclaration(None=FunctionInput(Preview=Constant() {r=1}) {bUsePreviewValueAsDefault=true, inputName=MinScale, inputType=FunctionInput_Scalar, sortPriority=2}) {name=MinScale, nodeColor=<rgba 1 0.0234 0 1>}
  NamedRerouteDeclaration(None=FunctionInput(Preview=Constant() {r=1.5}) {bUsePreviewValueAsDefault=true, inputName=MaxScale, inputType=FunctionInput_Scalar, sortPriority=2}) {name=MaxScale, nodeColor=<rgba 1 0.0234 0 1>}
  n24 = Constant3Vector() {constant=<rgba -1 -1 0 0>}
  NamedRerouteDeclaration(None=FunctionInput(Preview=ComponentMask(None=@n24) {g=true, r=true}) {bUsePreviewValueAsDefault=true, inputName=MinOffset, inputType=FunctionInput_Vector2, sortPriority=3}) {name=MinOffset, nodeColor=<rgba 0.1719 0 1 1>}
  n25 = Constant3Vector() {constant=<rgba 1 1 0 0>}
  NamedRerouteDeclaration(None=FunctionInput(Preview=ComponentMask(None=@n25) {g=true, r=true}) {bUsePreviewValueAsDefault=true, inputName=MaxOffset, inputType=FunctionInput_Vector2, sortPriority=3}) {name=MaxOffset, nodeColor=<rgba 0.1719 0 1 1>}
  NamedRerouteDeclaration(None=FunctionInput(Preview=Constant() {r=-90}) {bUsePreviewValueAsDefault=true, inputName="MinRotation(Degrees)", inputType=FunctionInput_Scalar, sortPriority=4}) {name=MinRotation, nodeColor=<rgba 0 1 0.3906 1>}
  NamedRerouteDeclaration(None=FunctionInput(Preview=Constant() {r=90}) {bUsePreviewValueAsDefault=true, inputName="MaxRotation(Degrees)", inputType=FunctionInput_Scalar, sortPriority=4}) {name=MaxRotation, nodeColor=<rgba 0 1 0.3906 1>}
  NamedRerouteDeclaration(None=@n23.R) {name=HexGradient_01, nodeColor=<rgba 1 0 0.8672 1>}
  NamedRerouteDeclaration(None=FunctionInput(Preview=TextureObject() {samplerType=SAMPLERTYPE_Normal, texture=/Engine/EngineMaterials/DefaultNormal}) {bUsePreviewValueAsDefault=true, inputName=Texture, inputType=FunctionInput_Texture2D}) {name=Texture_Source, nodeColor=<rgba 0 1 0.2031 1>}
  n26 = Multiply(A=@n4, B=FunctionInput(Preview=Constant() {r=1}) {bUsePreviewValueAsDefault=true, inputName=LowEndTiling, inputType=FunctionInput_Scalar})
  NamedRerouteDeclaration(None=TextureSample(UVs=Multiply(A=@n26) {constB=0.75}, Tex=NamedRerouteUsage()) {samplerSource=SSM_Wrap_WorldGroupSettings, samplerType=SAMPLERTYPE_Normal}.RGBA) {name=LowEnd_Output, nodeColor=<rgba 1 0 0.3984 1>}
  NamedRerouteUsage()
 ]
}
