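# Material utility function that maps an input UV onto a hexagonal grid and finds a cell center for it, filed under Asteria landscape materials.
# It rounds the UV against two offset rectangular lattices with 1 by √3 spacing, then picks the nearer candidate center using squared distances in a Custom node.
# The output packs that center's XY with only the X of its lattice index.
# Lerp, Sign/Max and UV-multiply branches are left unconnected.
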MaterialFunction /Game/Athena/Asteria/Environments/Landscape/Materials/MaterialFunctions/UtilityFunctions/MF_GetHexGridInfo
graph {nodes [
  n1 = TextureCoordinate()
  n2 = FunctionInput(Preview=@n1) {bUsePreviewValueAsDefault=true, inputName=uv, inputType=FunctionInput_Vector2}
  n3 = Subtract(A=@n2, B=Constant2Vector() {g=1, r=0.5})
  n4 = MaterialFunctionCall(X=Constant() {r=1}, Y=SquareRoot(None=Constant() {r=3})) {materialFunction=&MakeFloat2}
  n5 = Round(None=Divide(A=AppendVector(A=@n2, B=@n3), B=AppendVector(A=@n4.Result, B=@n4.Result)))
  n6 = ComponentMask(None=@n5) {a=true, b=true}
  n7 = AppendVector(A=Multiply(A=ComponentMask(None=@n5) {g=true, r=true}, B=@n4.Result), B=Multiply(A=Add(A=@n6) {constB=0.5}, B=@n4.Result))
  n8 = Subtract(A=AppendVector(A=@n2, B=@n2), B=@n7)
  n9 = ComponentMask(None=@n8) {g=true, r=true}
  MaterialExpressionDotProduct_0 = DotProduct(A=@n9, B=@n9)
  n10 = ComponentMask(None=@n8) {a=true, b=true}
  MaterialExpressionDotProduct_1 = DotProduct(A=@n10, B=@n10)
  n11 = Custom!(A=@MaterialExpressionDotProduct_0, B=@MaterialExpressionDotProduct_1) {
   code="return A < B;"
   inputs [
    {
     inputName=A
     input {expression=&self:MaterialExpressionDotProduct_0}
    }
    {
     inputName=B
     input {expression=&self:MaterialExpressionDotProduct_1}
    }
   ]
  }
  n12 = ComponentMask(None=@n11) {r=true}
  n13 = Constant()
  n14 = AppendVector(A=ComponentMask(None=@n7) {g=true, r=true}, B=ComponentMask(None=@n5) {g=true, r=true})
  n15 = AppendVector(A=ComponentMask(None=@n7) {a=true, b=true}, B=ComponentMask(None=@n5) {a=true, b=true})
  FunctionOutput(None=ComponentMask(None=If(A=@n12, B=@n13, "A > B"=@n14, "A == B"=@n15, "A < B"=@n15)) {b=true, g=true, r=true})
  LinearInterpolate(A=@n15, B=@n14, Alpha=@n12)
  Max(A=Sign(None=Subtract(A=@MaterialExpressionDotProduct_1, B=@MaterialExpressionDotProduct_0))) {constB=0}
  Multiply(A=@n1) {constB=3}
 ]}
