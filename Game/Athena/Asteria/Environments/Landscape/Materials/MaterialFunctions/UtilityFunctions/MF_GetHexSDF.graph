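# Landscape material utility function that returns a signed distance to a hexagon centered at the origin, for a 2D point p.
# It folds p into one quadrant with Abs, then takes the larger of its X and its projection onto the normalized (1, √3) edge normal.
# Subtracting that from 0.5 gives positive values inside a pointy-top hexagon of unit width and negative values outside.
# Texture coordinates serve as the default input.
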MaterialFunction /Game/Athena/Asteria/Environments/Landscape/Materials/MaterialFunctions/UtilityFunctions/MF_GetHexSDF
graph {nodes [
  n1 = Constant() {r=1}
  n2 = SquareRoot(None=Constant() {r=3})
  n3 = Abs(None=FunctionInput(Preview=TextureCoordinate()) {bUsePreviewValueAsDefault=true, description=p, inputType=FunctionInput_Vector2})
  n4 = DotProduct(
   A=Multiply(A=MaterialFunctionCall(X=@n1, Y=@n2) {materialFunction=&MakeFloat2}.Result) {constB=0.5}
   B=@n3
  )
  n5 = ComponentMask(None=@n3) {r=true}
  FunctionOutput(None=Subtract(A=Constant() {r=0.5}, B=Max(A=@n4, B=@n5)))
 ]}
