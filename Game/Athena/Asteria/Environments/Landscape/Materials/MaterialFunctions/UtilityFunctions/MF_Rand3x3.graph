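# MF_Rand3x3 is a landscape utility material function that hashes a 3-component input into a pseudo-random 3-component value in the 0 to 1 range.
# It dots the input with three different constant vectors, packs the results with MakeFloat3, then applies sine, a large multiplier and Frac.
# The input previews as texture coordinates padded with a constant.
# Leftover nodes from scalar and 2D hash variants do not reach the output.
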MaterialFunction /Game/Athena/Asteria/Environments/Landscape/Materials/MaterialFunctions/UtilityFunctions/MF_Rand3x3
graph {nodes [
  Constant2Vector() {g=269.5, r=127.1}
  Constant2Vector() {g=183.3, r=311.7}
  n1 = FunctionInput(Preview=AppendVector(A=TextureCoordinate(), B=Constant()))
  n2 = DotProduct(A=Constant2Vector() {g=4.1414, r=12.9898}, B=@n1)
  Frac(None=Multiply(A=Sine(None=@n2), B=Constant() {r=43758.5469}))
  Frac(None=Multiply(A=Sine(), B=Constant() {r=43758.5469}))
  n3 = Constant3Vector() {constant=<rgba 127.1 311.7 74.7 0>}
  n4 = Constant3Vector() {constant=<rgba 269.5 183.3 246.1 0>}
  n5 = Constant3Vector() {constant=<rgba 113.5 271.9 124.6 0>}
  n6 = Sine(None=MaterialFunctionCall(X=DotProduct(A=@n1, B=@n3), Y=DotProduct(A=@n1, B=@n4), Z=DotProduct(A=@n1, B=@n5)) {materialFunction=&MakeFloat3}.Result)
  n7 = Constant() {r=43758.5469}
  FunctionOutput(None=Frac(None=Multiply(A=@n6, B=@n7)))
  MaterialFunctionCall()
 ]}
