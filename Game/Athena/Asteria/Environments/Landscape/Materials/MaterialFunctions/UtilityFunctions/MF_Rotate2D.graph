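# Landscape utility material function that rotates a 2D point, texture coordinates by default, around a pivot that defaults to the UV center.
# It subtracts the pivot, applies the standard sine and cosine rotation to X and Y, recombines them with MakeFloat2 and adds the pivot back.
# Sine and Cosine keep their default period, so the angle input is in full turns rather than radians.
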MaterialFunction /Game/Athena/Asteria/Environments/Landscape/Materials/MaterialFunctions/UtilityFunctions/MF_Rotate2D
graph {nodes [
  n1 = FunctionInput(Preview=Constant2Vector() {g=0.5, r=0.5}) {bUsePreviewValueAsDefault=true, inputName=pivotPoint, inputType=FunctionInput_Vector2}
  n2 = Subtract(A=FunctionInput(Preview=TextureCoordinate()) {bUsePreviewValueAsDefault=true, inputName=inPoint, inputType=FunctionInput_Vector2}, B=@n1)
  n3 = ComponentMask(None=@n2) {desc=X, r=true}
  n4 = FunctionInput(Preview=ScalarParameter() {parameterName=Angle}) {bUsePreviewValueAsDefault=true, inputName=angle, inputType=FunctionInput_Scalar}
  n5 = Cosine(None=@n4) {desc="Cos(theta)"}
  n6 = ComponentMask(None=@n2) {desc=Y, g=true}
  n7 = Sine(None=@n4) {desc="Sine(theta)"}
  n8 = Subtract(A=Multiply(A=@n3, B=@n5) {desc="X * Cos(theta)"}, B=Multiply(A=@n6, B=@n7) {desc="Y * Sine(Theta)"}) {desc=New X}
  n9 = Add(A=Multiply(A=@n3, B=@n7) {desc="X * Sin(theta)"}, B=Multiply(A=@n6, B=@n5) {desc="Y * Cos(theta)"}) {desc=NewY}
  FunctionOutput(None=Add(A=MaterialFunctionCall(X=@n8, Y=@n9) {materialFunction=&MakeFloat2}.Result, B=@n1))
 ]}
