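# MF_Trinterp is a landscape material utility function that evaluates an offsettable hexagonal grid over a UV input.
# It scales UV by freq, adds Offset divided by half of the (1, √3) hex ratio, and feeds that to MF_GetHexGridInfo.
# The returned cell XY is subtracted from the coordinate for MF_GetHexSDF.
# It outputs a three-channel vector: cell XY divided by freq, plus the doubled hex distance.
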MaterialFunction /Game/Athena/Asteria/Environments/Landscape/Materials/MaterialFunctions/UtilityFunctions/MF_Trinterp
graph {nodes [
  n1 = Constant2Vector()
  n2 = MaterialFunctionCall(X=Constant() {r=1}, Y=SquareRoot(None=Constant() {r=3})) {materialFunction=&MakeFloat2}
  n3 = TextureCoordinate()
  n4 = FunctionInput(Preview=Constant() {r=8}) {bUsePreviewValueAsDefault=true, inputName=freq, inputType=FunctionInput_Scalar}
  n5 = Add(
   A=Divide(
    A=FunctionInput(Preview=@n1) {bUsePreviewValueAsDefault=true, inputName=Offset, inputType=FunctionInput_Vector2}
    B=Multiply(A=@n2.Result) {constB=0.5}
   )
   B=Multiply(A=FunctionInput(Preview=@n3) {bUsePreviewValueAsDefault=true, inputName=uv, inputType=FunctionInput_Vector2}, B=@n4)
  )
  n6 = MaterialFunctionCall!(uv=@n5) {materialFunction=&/MF_GetHexGridInfo}
  n7 = Divide(A=@n6.Result, B=@n4)
  n8 = MaterialFunctionCall!(In=Subtract(A=@n5, B=ComponentMask(None=@n6.Result) {g=true, r=true})) {materialFunction=&/MF_GetHexSDF}
  FunctionOutput(None=AppendVector(A=ComponentMask(None=@n7) {g=true, r=true}, B=Multiply(A=@n8.Result) {constB=2}))
 ]}
