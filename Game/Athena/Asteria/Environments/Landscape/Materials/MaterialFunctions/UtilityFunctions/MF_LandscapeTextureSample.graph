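# Landscape utility function that samples a texture with randomized, blended tiling, apparently to hide visible repetition.
# Three offset MF_Trinterp evaluations supply cells and blend weights; MF_Rand3x3 turns each cell into a random rotation and UV shift.
# The three rotated samples are weighted and summed, alongside a cheaper single scaled sample, with shading-path and quality switches choosing the output for mobile versus other paths.
# A debug output exposes the first blend weight.
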MaterialFunction /Game/Athena/Asteria/Environments/Landscape/Materials/MaterialFunctions/UtilityFunctions/MF_LandscapeTextureSample
graph {
 properties {bExposeToLibrary=true}
 nodes [
  n1 = NamedRerouteUsage()
  n2 = NamedRerouteUsage()
  FunctionOutput(None=ShadingPathSwitch(Default=QualitySwitch(Default=@n1), Mobile=QualitySwitch(Default=@n2, Low=@n2, Medium=@n2)))
  n3 = FunctionInput(Preview=TextureCoordinate()) {bUsePreviewValueAsDefault=true, inputName=LandscapeCoordinates, inputType=FunctionInput_Vector2}
  n4 = FunctionInput(Preview=ScalarParameter() {defaultValue=1, parameterName=TileFrequency}) {bUsePreviewValueAsDefault=true, inputName=Tiling, inputType=FunctionInput_Vector2}
  n5 = MaterialFunctionCall!(uv=@n3, freq=@n4, Offset) {materialFunction=&/MF_Trinterp}
  FunctionOutput(None=NamedRerouteDeclaration(None=ComponentMask(None=@n5.Result) {b=true}) {name=weight0, nodeColor=<rgba 0.6406 1 0 1>}) {outputName=DebugOutput}
  n6 = ScalarParameter() {defaultValue=1, parameterName=rotationMult}
  n7 = MaterialFunctionCall!(In=AppendVector(A=ComponentMask(None=@n5.Result) {g=true, r=true}, B=Constant())) {materialFunction=&/MF_Rand3x3}
  n8 = ComponentMask(None=@n7.Result) {b=true}
  n9 = Multiply(A=@n3, B=Multiply(A=ScalarParameter() {defaultValue=0.7, parameterName=InternalFrequency}, B=@n4))
  n10 = Add(
   A=MaterialFunctionCall!(pivotPoint, angle=Multiply(A=@n6, B=@n8), inPoint=NamedRerouteDeclaration(None=@n9) {name=pointToRotate, nodeColor=<rgba 0.5234 0 1 1>}) {materialFunction=&/MF_Rotate2D}.Result
   B=ComponentMask(None=@n7.Result) {g=true, r=true}
  )
  n11 = Multiply(
   A=TextureSample(UVs=NamedRerouteDeclaration(None=@n10) {name=uv0, nodeColor=<rgba 0 0.9062 1 1>}, Tex=NamedRerouteUsage()) {samplerSource=SSM_Wrap_WorldGroupSettings}.RGBA
   B=NamedRerouteUsage()
  )
  n12 = MaterialFunctionCall(uv=@n3, freq=@n4, Offset=Constant2Vector() {g=0.25, r=0.25}) {materialFunction=&MF_Trinterp}
  n13 = MaterialFunctionCall(In=AppendVector(A=ComponentMask(None=@n12.Result) {g=true, r=true}, B=Constant())) {materialFunction=&MF_Rand3x3}
  n14 = ComponentMask(None=@n13.Result) {b=true}
  n15 = Add(
   A=MaterialFunctionCall(angle=Multiply(A=@n6, B=@n14), inPoint=NamedRerouteUsage()) {materialFunction=&MF_Rotate2D}.Result
   B=ComponentMask(None=@n13.Result) {g=true, r=true}
  )
  n16 = Multiply(
   A=TextureSample(UVs=NamedRerouteDeclaration(None=@n15) {name=uv1, nodeColor=<rgba 0 0.7422 1 1>}, Tex=NamedRerouteUsage()) {samplerSource=SSM_Wrap_WorldGroupSettings}.RGBA
   B=NamedRerouteUsage()
  )
  n17 = MaterialFunctionCall(uv=@n3, freq=@n4, Offset=Constant2Vector() {g=-0.25, r=0.25}) {materialFunction=&MF_Trinterp}
  n18 = MaterialFunctionCall(In=AppendVector(A=ComponentMask(None=@n17.Result) {g=true, r=true}, B=Constant())) {materialFunction=&MF_Rand3x3}
  n19 = MaterialFunctionCall(angle=Multiply(A=ComponentMask(None=@n18.Result) {b=true}, B=@n6), inPoint=NamedRerouteUsage()) {materialFunction=&MF_Rotate2D}
  n20 = ComponentMask(None=@n18.Result) {g=true, r=true}
  n21 = TextureSample(UVs=NamedRerouteDeclaration(None=Add(A=@n19.Result, B=@n20)) {name=uv2, nodeColor=<rgba 0 1 0.1562 1>}, Tex=NamedRerouteUsage()) {samplerSource=SSM_Wrap_WorldGroupSettings}
  n22 = NamedRerouteUsage()
  NamedRerouteDeclaration(None=Add(A=Add(A=@n11, B=@n16), B=Multiply(A=@n21.RGBA, B=@n22))) {name=HighEnd_Output, nodeColor=<rgba 0 0.4141 1 1>}
  NamedRerouteDeclaration(None=ComponentMask(None=@n12.Result) {b=true}) {name=weight1, nodeColor=<rgba 0.8984 1 0 1>}
  NamedRerouteDeclaration(None=ComponentMask(None=@n17.Result) {b=true}) {name=weight2, nodeColor=<rgba 0 0.1094 1 1>}
  NamedRerouteDeclaration(None=FunctionInput(Preview=TextureObject() {samplerType=SAMPLERTYPE_Masks, texture=/Game/Environments/Apollo/Sets/Terrain/Textures/T_Apollo_Grass_D_02}) {bUsePreviewValueAsDefault=true, inputName=Texture, inputType=FunctionInput_Texture2D}) {name=Texture_Source, nodeColor=<rgba 0 1 0.2031 1>}
  n23 = Multiply(A=@n3, B=@n4)
  NamedRerouteDeclaration(None=TextureSample(UVs=Multiply(A=@n23) {constB=0.75}, Tex=NamedRerouteUsage()) {samplerSource=SSM_Wrap_WorldGroupSettings}.RGBA) {name=LowEnd_Output, nodeColor=<rgba 1 0 0.3984 1>}
  NamedRerouteUsage()
 ]
}
